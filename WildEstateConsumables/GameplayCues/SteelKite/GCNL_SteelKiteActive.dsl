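; Looping gameplay cue for the active Steel Kite effect, driving its particles, audio and water-splash feedback on the affected pawn.
; Application fades in a post-process for the local viewer and starts a timer polling remaining duration to warn of expiry.
; Footstep, landing and slide-timer events trace for water during tactical sprint, toggling splash Niagara, dust and loose water tags.
; Removal fades audio and screen effect, unbinds events and clears water tags.
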(event Custom|StartAboutToWearoffTimer
  ;[0.004566 0.1304 0.8385 0.5] Timer and NS param set to alert players that the effect will wear off soon
  (bind _returnvalue (GameplayEffects|GetActiveEffectswithAllTags (Ability|GetAbilitySystemComponent (Variables|Default|GetPlayerPawn)) (Variables|Default|GetGETagsToGetActiveHandle)))
  (Utilities|Time|SetTimerbyEvent (delegate Custom|Check%20for%20Expiry%20Effect%20Add) 1.0 true))

(event EventBeginPlay
  (bind _returnvalue (Rendering|Material|CreateDynamicMaterialInstance 0 "/WildEstateConsumables/Gameplay/SteelKite/FX/PostProcess/PPMI_StreetKite.PPMI_StreetKite" "None" "None" :!class /Script/Engine.KismetMaterialLibrary))
  (bind _output_get (Variables|Default|SetDM _returnvalue))
  (Variables|/Script/Engine.PostProcessComponent|SetSettings (Utilities|Struct|MakePostProcessSettings (Utilities|Struct|MakeWeightedBlendables (Utilities|Array|MakeArray (Utilities|Struct|MakeWeightedBlendable 1.0 _output_get)))) (Variables|Default|GetPostProcess)))

(event Custom|Check%20for%20Expiry%20Effect%20Add
  (bind _returnvalue 5.0)
  (bind _playerpawn (Variables|Default|GetPlayerPawn))
  (bind _active_looping_particle (Variables|Default|GetActive%20Looping%20Particle))
  ;[0.004566 0.1304 0.8385 0.5] Timer and NS param set to alert players that the effect will wear off soon
  (bind _returnvalue_2 (Ability|GameplayEffect|GetActiveGameplayEffectRemainingDuration 0 (Utilities|Array|Get(acopy) (stale Custom|StartAboutToWearoffTimer _returnvalue))))
  (if (<= _returnvalue_2 _returnvalue)
    (Effects|Components|ParticleSystem|SetBoolParameter _active_looping_particle "PlayRemoveSoon" true)
    (Effects|Components|ParticleSystem|SetFloatParameter _active_looping_particle "RemoveSoonTime" _returnvalue)
    (if (Pawn|IsLocallyViewed _playerpawn)
      (Utilities|FlowControl|DoOnce false
        (:Completed
          (Audio|SpawnSoundAttached (Variables|Audio|GetEndingSoonSound) (Variables|/Script/Engine.Actor|GetRootComponent _playerpawn)))))
    (else
      (Effects|Components|ParticleSystem|SetBoolParameter _active_looping_particle "PlayRemoveSoon" false))))

(event Collision|EventActorBeginOverlap
  (params (OtherActor :type Actor%20Object%20Reference)))

(event EventTick
  (params (DeltaSeconds :type Float%20%28single-precision%29)))

(event EventOnApplication(EitherVFX-preferthisone)
  (params (MyTarget :type Actor%20Object%20Reference) (Parameters :type Gameplay%20Cue%20Parameters%20Structure%20%28by%20ref%29) (ParticleComponents :type Array%20of%20FXSystem%20Component%20Object%20References) (AudioComponents :type Array%20of%20Audio%20Component%20Object%20References) (BurstCameraShakeInstance :type Camera%20Shake%20Base%20Object%20Reference) (BurstDecalInstance :type Decal%20Actor%20Object%20Reference))
  (Rendering|SetRenderCustomDepth (Utilities|Array|Get(acopy) ParticleComponents) true)
  (bind _asfort_player_pawn (Utilities|Casting|CastToFortPlayerPawn MyTarget))
  (Variables|Default|SetPlayerPawn _asfort_player_pawn)
  (Utilities|FlowControl|Sequence
    (:then_0
      (if (Pawn|IsLocallyViewed (Variables|Default|GetPlayerPawn))
        (Variables|/Script/Engine.PostProcessComponent|SetbEnabled true (Variables|Default|GetPostProcess))
        (Variables|/Script/Engine.PostProcessComponent|SetBlendWeight 1.0 (Variables|Default|GetPostProcess))
        (Rendering|Material|SetScalarParameterValue (Variables|Default|GetDM) "CenterMaskSize" 0.902 :!class /Script/Engine.MaterialInstanceDynamic)
        (bind (_direction _fade) (Timeline|FadeIN PlayFromStart)
          (:Update
            (Rendering|Material|SetScalarParameterValue (Variables|Default|GetDM) "Mask_Lerp" _fade :!class /Script/Engine.MaterialInstanceDynamic)))))
    (:then_1
      (CallFunction|StartAbouttoWearoffTimer))))

(event EventOnLoopingStart(EitherVFX-preferthisversion)
  (params (MyTarget :type Actor%20Object%20Reference) (Parameters :type Gameplay%20Cue%20Parameters%20Structure%20%28by%20ref%29) (ParticleComponents :type Array%20of%20FXSystem%20Component%20Object%20References) (AudioComponents :type Array%20of%20Audio%20Component%20Object%20References))
  (bind _fortplayerpawn (Variables|Default|GetFortPlayerPawn))
  (|CallParentFunction MyTarget Parameters ParticleComponents AudioComponents)
  (bind _asfort_player_pawn (Utilities|Casting|CastToFortPlayerPawn MyTarget))
  (bind _output_get (Variables|Default|SetFortPlayerPawn _asfort_player_pawn))
  (bind _output_get_2 (Variables|Default|SetActive%20Looping%20Particle (Utilities|Array|Get(acopy) ParticleComponents)))
  (Rendering|SetRenderCustomDepth _output_get_2 true)
  (Variables|Default|SetSpawnedParticleComponents ParticleComponents)
  (Utilities|FlowControl|Sequence
    (:then_0
      (EventDispatchers|BindEventtoOnFootstepEvent _fortplayerpawn (EventDispatchers|CreateEvent :!event WaterFootsteps))
      (EventDispatchers|BindEventtoOnPawnLanded _fortplayerpawn (EventDispatchers|CreateEvent :!event Custom|LandedCheckWater))
      (bind _asynctaskproxy (Ability|Tasks|WaitGameplayTagAddToActor _fortplayerpawn (Variables|Default|GetSliding%20Tag))
        (:then
          (Variables|Default|SetAsync%20Slide%20Tag%20Action _asynctaskproxy))
        (:Added
          (bind _returnvalue (Utilities|Time|SetTimerbyEvent (delegate Custom|SlidingLoopingTimerEvent) 0.15 true true))
          (Variables|Default|SetSlidingLoopingTimerHandle _returnvalue))))
    (:then_1
      ;[1 1 1] Trigger activated audio loop
      (bind _returnvalue_2 (Audio|SpawnSoundAttached (select (Pawn|IsLocallyViewed _output_get) (Variables|Audio|GetActivatedLoop1p) (Variables|Audio|GetActivatedLoop3p)) (Variables|/Script/Engine.Actor|GetRootComponent _output_get)))
      (Variables|Default|SetActivatedLoopComp _returnvalue_2))))

(event Custom|SlidingLoopingTimerEvent
  (CallFunction|LandedCheckWater))

(event EventOnRemoval(EitherVFX-preferthisversion)
  (params (MyTarget :type Actor%20Object%20Reference) (Parameters :type Gameplay%20Cue%20Parameters%20Structure%20%28by%20ref%29) (ParticleComponents :type Array%20of%20FXSystem%20Component%20Object%20References) (AudioComponents :type Array%20of%20Audio%20Component%20Object%20References) (BurstCameraShakeInstance :type Camera%20Shake%20Base%20Object%20Reference) (BurstDecalInstance :type Decal%20Actor%20Object%20Reference))
  (bind _fortplayerpawn (Variables|Default|GetFortPlayerPawn))
  (Utilities|FlowControl|Sequence
    (:then_0
      (Audio|Components|Audio|FadeOut (Variables|Default|GetActivatedLoopComp) 0.5 0.0 "Linear" :!class /Script/Engine.AudioComponent)
      (if (Pawn|IsLocallyViewed (Variables|Default|GetPlayerPawn))
        (Audio|SpawnSound2D 0 (Variables|Audio|GetEndingSound))
        (bind (_direction _centermasksize) (Timeline|Fadeout PlayFromStart)
          (:Update
            (Rendering|Material|SetScalarParameterValue (Variables|Default|GetDM) "CenterMaskSize" _centermasksize :!class /Script/Engine.MaterialInstanceDynamic))
          (:Finished
            (Variables|/Script/Engine.PostProcessComponent|SetbEnabled false (Variables|Default|GetPostProcess))))))
    (:then_1
      (|CallParentFunction MyTarget Parameters ParticleComponents AudioComponents BurstCameraShakeInstance BurstDecalInstance)
      (EventDispatchers|UnbindEventfromOnFootstepEvent _fortplayerpawn (EventDispatchers|CreateEvent :!event WaterFootsteps))
      (EventDispatchers|UnbindEventfromOnPawnLanded _fortplayerpawn (EventDispatchers|CreateEvent :!event Custom|LandedCheckWater))
      (Ability|Async|EndAction (Variables|Default|GetAsync%20Slide%20Tag%20Action))
      (Utilities|Time|ClearandInvalidateTimerbyHandle self (Variables|Default|GetSlidingLoopingTimerHandle))
      (bind _playerpawn (Variables|Default|GetPlayerPawn)
        (:then
          (if (GameplayTags|HasMatchingGameplayTag _playerpawn (Variables|Default|GetWater%20Tag%20to%20Check))
            (Ability|GameplayEffect|RemoveLooseGameplayTags _playerpawn (Variables|Default|GetWater%20Gameplay%20Tags))))
        (:else)))))

(event Custom|LandedCheckWater
  (params (Hit :type Hit%20Result%20Structure%20%28by%20ref%29))
  (Default|WaterFootsteps 0))

(fn ConstructionScript)

(fn WaterFootsteps
  (params (FootIndex :type Integer))
  (bind _fortplayerpawn (Variables|Default|GetFortPlayerPawn))
  (bind _mesh (Variables|/Script/Engine.Character|GetMesh _fortplayerpawn))
  (bind _returnvalue (Transformation|GetSocketTransform _mesh "ball_r" "RTS_World"))
  (bind _returnvalue_2 (Transformation|GetSocketTransform _mesh "ball_l" "RTS_World"))
  (bind _returnvalue_3 (Math|Transform|SelectTransform _returnvalue _returnvalue_2 (== FootIndex 1)))
  (bind _location (.location _returnvalue_3))
  (bind _returnvalue_4 (== FootIndex 1))
  (bind _footstepwatersplashfx (Variables|Default|GetFootstepWaterSplashFX))
  (if (Variables|/Script/FortniteGame.FortPlayerPawn|GetbIsTacticalSprinting (Variables|Default|GetFortPlayerPawn))
    (bind (_outhit _returnvalue_5) (Collision|LineTraceForObjects (Variables|Default|GetFortPlayerPawn) (+ _location "0.000000,0.000000,100.000000") (+ _location "0.000000,0.000000,-50.000000") (Variables|Default|GetObject%20Types)))
    (Variables|/Script/FortniteGame.FortPlayerPawn|SetbFXPlayDustOnMovement (not _returnvalue_5) (Variables|Default|GetFortPlayerPawn))
    (if _returnvalue_5
      (Variables|Default|GetFootstepWaterSplashFX
        (:then
          (Components|Activation|Activate _footstepwatersplashfx false)
          (Rendering|SetVisibility _footstepwatersplashfx true true)
          (Transformation|SetWorldLocation _footstepwatersplashfx (.Location (Collision|BreakHitResult _outhit)) false false))
        (:else
          (bind _returnvalue_6 (Niagara|SpawnSystematLocation :SystemTemplate (Variables|Default|GetFootstep%20Water%20System%20Template) :Location (.Location (Collision|BreakHitResult _outhit)) :Scale "1.000000,1.000000,1.000000" :bAutoDestroy true :bAutoActivate true :PoolingMethod "AutoRelease" :bPreCullCheck true))
          (Variables|Default|SetFootstepWaterSplashFX _returnvalue_6)))
      (Niagara|NiagaraSetVector4Array (Variables|Default|GetFootstepWaterSplashFX) "SpawnList" (Utilities|Array|MakeArray (Math|Vector4|MakeVector4 (.x _location) (.y _location) (.z _location) 1.0)))
      (bind _playerpawn (Variables|Default|GetPlayerPawn)
        (:then
          (if (not (GameplayTags|HasMatchingGameplayTag _playerpawn (Variables|Default|GetWater%20Tag%20to%20Check)))
            (Ability|GameplayEffect|AddLooseGameplayTags _playerpawn (Variables|Default|GetWater%20Gameplay%20Tags))))
        (:else))
      (else
        (Components|Activation|Deactivate (Variables|Default|GetFootstepWaterSplashFX))
        (Variables|Default|SetFootstepWaterSplashFX 0)
        (bind _playerpawn_2 (Variables|Default|GetPlayerPawn)
          (:then
            (if (GameplayTags|HasMatchingGameplayTag _playerpawn_2 (Variables|Default|GetWater%20Tag%20to%20Check))
              (Ability|GameplayEffect|RemoveLooseGameplayTags _playerpawn_2 (Variables|Default|GetWater%20Gameplay%20Tags))))
          (:else))))))
