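; Gameplay cue notify for consuming the SteelKite consumable, currently holding only debug output in its graph.
; The EitherVFX OnBurst event casts the target to FortPlayerPawn and prints the pawn's display name, or "Cast Failed" when the cast fails.
; The Niagara OnBurst variant is empty, and neither event uses the supplied particle, audio, camera shake or decal references.
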(fn OnBurst(EitherVFX-preferthisversion)
  (params (MyTarget :type Actor%20Object%20Reference) (Parameters :type Gameplay%20Cue%20Parameters%20Structure%20%28by%20ref%29) (ParticleComponents :type Array%20of%20FXSystem%20Component%20Object%20References) (AudioComponents :type Array%20of%20Audio%20Component%20Object%20References) (BurstCameraShakeInstance :type Camera%20Shake%20Base%20Object%20Reference) (BurstDecalInstance :type Decal%20Actor%20Object%20Reference))
  (bind _asfort_player_pawn (Utilities|Casting|CastToFortPlayerPawn MyTarget)
    (:then
      (Development|PrintString 0 (Utilities|GetDisplayName _asfort_player_pawn) true true "(R=1.000000,G=0.854573,B=0.000000,A=1.000000)"))
    (:CastFailed
      (Development|PrintString 0 "Cast Failed" true true "(R=1.000000,G=0.911157,B=0.000000,A=1.000000)")))
  (return))

(fn OnBurst(Niagara)
  (params (MyTarget :type Actor%20Object%20Reference) (Parameters :type Gameplay%20Cue%20Parameters%20Structure%20%28by%20ref%29) (ParticleComponents :type Array%20of%20Niagara%20Particle%20System%20Component%20Object%20References) (AudioComponents :type Array%20of%20Audio%20Component%20Object%20References) (BurstCameraShakeInstance :type Camera%20Shake%20Base%20Object%20Reference) (BurstDecalInstance :type Decal%20Actor%20Object%20Reference))
  (return))
